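# Sound cue for the bounce of the Artemis mushroom launcher environment actor, named as the first-person (1P) variant.
# A random node picks one of four bounce waves and feeds it through a modulator at default settings into a mixer, whose second input is empty.
# It plays in the gadgets and items sound class, with its own attenuation and per-owner concurrency overrides that stop the oldest instance at the limit.
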SoundCue /FNGameplayCues/Sounds/Artemis/Environment_Actors/Mushroom_Bouncer/Env_Mushroom_Launcher_Bounce_1P_Cue
root=Mixer_0
nodes [
 {id=Mixer_0, children [Modulator_0,null], properties {inputVolume [1,1]}}
 {id=Modulator_0, children [Random_1]}
 {id=Random_1, children [WavePlayer_6,WavePlayer_7,WavePlayer_8,WavePlayer_9], properties {weights [1,1,1,1]}}
 {
  id=WavePlayer_6
  properties {soundWaveAssetPtr=&/Env_Mushroom_Launcher_Bounce_04}
 }
 {
  id=WavePlayer_7
  properties {soundWaveAssetPtr=&/Env_Mushroom_Launcher_Bounce_01}
 }
 {
  id=WavePlayer_8
  properties {soundWaveAssetPtr=&/Env_Mushroom_Launcher_Bounce_02}
 }
 {
  id=WavePlayer_9
  properties {soundWaveAssetPtr=&/Env_Mushroom_Launcher_Bounce_03}
 }
]
properties {
 bOverrideAttenuation=true
 bOverrideConcurrency=true
 maxDistance=2000
 priority=2
 soundClassObject=&Fort_Gadgets_Items
 volumeMultiplier=0.65
 attenuationOverrides {falloffDistance=1500, attenuationShapeExtents {x=500}}
 concurrencyOverrides {bLimitToOwner=true, maxCount=3, resolutionRule=StopOldest, retriggerTime=1}
 concurrencySet [null]
}
